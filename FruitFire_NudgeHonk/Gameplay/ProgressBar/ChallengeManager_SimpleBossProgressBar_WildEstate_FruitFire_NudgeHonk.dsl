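; Challenge manager Blueprint that drives a simple boss progress bar for the WildEstate boss in FruitFire NudgeHonk.
; On AI spawn it registers the boss, sets its leash and music tag, binds death, and starts the challenge on first damage.
; HandleBossHealthChanged blends health and shield percents into normalized progress, weighted by scalable floats indexed by the playlist's clamped max team size.
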(event EventBeginPlay
  (bind (_gamestate _playlist) (Class|FortAsyncActionCurrentPlaylistReady|PlaylistDataReadyAsync)
    (:OnReady
      (Variables|Default|SetPlaylistMaxTeamSize (Math|Integer|Clamp(Integer) (Variables|/Script/FortniteGame.FortPlaylist|GetMaxTeamSize _playlist) 1 4)))))

(event Collision|EventActorBeginOverlap
  (params (OtherActor :type Actor%20Object%20Reference)))

(event EventTick
  (params (DeltaSeconds :type Float%20%28single-precision%29)))

(event EventAISpawned
  (params (SpawnedActor :type Actor%20Object%20Reference) (bResult :type Boolean))
  (bind _asfort_player_pawn_athena (Utilities|Casting|CastToFortPlayerPawnAthena SpawnedActor))
  (EventDispatchers|BindEventtoOnDamaged _asfort_player_pawn_athena (delegate Custom|SilentNovel_VaneHush_OnDamaged))
  (CallFunction|HandleBossSpawned SpawnedActor)
  (CallFunction|SetIsBossReadyforFinalFight true)
  (Leash|UpdateAIActorLeash SpawnedActor)
  (Music|AddCurrentMusicTagtoActor SpawnedActor)
  (bind _asfort_player_pawn_athena_2 (Utilities|Casting|CastToFortPlayerPawnAthena SpawnedActor))
  (bind _asfort_athena_aibot_controller (Utilities|Casting|CastToFortAthenaAIBotController (Pawn|GetController _asfort_player_pawn_athena_2)))
  (Death|BindEventtoOnAgentDiedEvent _asfort_athena_aibot_controller (EventDispatchers|CreateEvent :!event AIDied))
  (CallFunction|HandleBossHealthChanged))

(event Custom|SilentNovel_VaneHush_OnDamaged
  (params (DamagedActor :type Actor%20Object%20Reference) (Damage :type Float%20%28single-precision%29) (InstigatedBy :type Controller%20Object%20Reference) (DamageCauser :type Actor%20Object%20Reference) (HitLocation :type Vector) (FHitComponent :type Primitive%20Component%20Object%20Reference) (BoneName :type Name) (Momentum :type Vector))
  (bind _asfort_player_pawn_athena (Utilities|Casting|CastToFortPlayerPawnAthena (Variables|Boss|GetBossActor)))
  (EventDispatchers|UnbindEventfromOnDamaged _asfort_player_pawn_athena (delegate Custom|SilentNovel_VaneHush_OnDamaged))
  (CallFunction|StartSimpleChallenge))

(fn HandleBossHealthChanged
  (bind _bossactor (Variables|Boss|GetBossActor))
  (if (Variables|Default|GetBossIsReadyForFinalFight)
    (Utilities|IsValid _bossactor
      (:Is%20Valid
        (bind _asbp_player_pawn_non_participant_boss (Utilities|Casting|CastToBP_PlayerPawn_NonParticipant_Boss _bossactor))
        (Variables|Progress|SetBossProgress (Utilities|NormalizeFloat (+ (* (FortPawn|HealthPercent _asbp_player_pawn_non_participant_boss) (ScalableFloat|GetValueatLevel (Variables|Default|GetMaxHealthScalar) (Math|Conversions|ToFloat(Integer) (- (Variables|Default|GetPlaylistMaxTeamSize) 1)))) (* (FortPawn|ShieldPercent _asbp_player_pawn_non_participant_boss) (ScalableFloat|GetValueatLevel (Variables|Default|GetMaxShieldScalar) (Math|Conversions|ToFloat(Integer) (- (Variables|Default|GetPlaylistMaxTeamSize) 1))))) 0 (+ (ScalableFloat|GetValueatLevel (Variables|Default|GetMaxHealthScalar) (Math|Conversions|ToFloat(Integer) (- (Variables|Default|GetPlaylistMaxTeamSize) 1))) (ScalableFloat|GetValueatLevel (Variables|Default|GetMaxShieldScalar) (Math|Conversions|ToFloat(Integer) (- (Variables|Default|GetPlaylistMaxTeamSize) 1)))) true))))))

(fn ConstructionScript
  (|CallParentFunction))
