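# Translucent UI material that draws the FruitFire NudgeHonk progress bar portrait as an icon over a distorted circular badge.
# The portrait texture gets parameterized offset and centered scale, with squared RGB as color and alpha as mask.
# A distortion function warps the scaled UVs meant for two MF_UI_Circle calls, which yield inner and outer fills and outlines.
# Named reroutes carry these layers into lerped color with a tintable outline, and a mask with circle and master opacity.
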Material /FruitFire_NudgeHonk/Gameplay/ProgressBar/Materials/M_UI_FruitFire_NudgeHonk_Portrait
graph {
 properties {bAutomaticallySetUsageInEditor=false, blendMode=BLEND_Translucent, materialDomain=MD_UI, usedShadingModels=Default Lit}
 nodes [
  n1 = TextureCoordinate()
  n2 = Add(A=ScalarParameter() {defaultValue=0.05, parameterName=Icon X pos}, B=ComponentMask(None=@n1) {r=true}) {constB=0}
  n3 = Add(A=ComponentMask(None=@n1) {g=true}, B=ScalarParameter() {defaultValue=0.06, parameterName=Icon Y pos}) {constB=0.06}
  n4 = TextureSample(UVs=MaterialFunctionCall(UVs=AppendVector(A=@n2, B=@n3), Texture Scale=ScalarParameter() {defaultValue=0.8, parameterName=Icon Scale}) {materialFunction=&ScaleUVsByCenter}.UVs) {samplerSource=SSM_Clamp_WorldGroupSettings, texture=/FruitFire_NudgeHonk/Gameplay/ProgressBar/Texture/T_UI_FruitFire_NudgeHonk_Portrait}
  NamedRerouteDeclaration(None=@n4.A) {name=IconMask, nodeColor=<rgba 0 0 0 1>}
  n5 = Add(
   A=Subtract(A=TextureCoordinate()) {constB=0.5}
   B=MaterialFunctionCall!(UV, SpeedX=Constant(), SpeedY=Constant(), Intensity=ScalarParameter() {defaultValue=0.06, parameterName=Deformation intensity}) {materialFunction=&/MF_UI_WildEstate_Distort_2}.Result
  )
  n6 = TextureCoordinate()
  n7 = AppendVector(A=Add(B=ComponentMask(None=@n6) {r=true}) {constB=0}, B=Add(A=ComponentMask(None=@n6) {g=true}) {constB=-0.3})
  NamedRerouteDeclaration(None=MaterialFunctionCall(UVs=Add(A=@n5, B=@n7), Texture Scale=ScalarParameter() {defaultValue=1.5, parameterName=Circle scale}) {materialFunction=&ScaleUVsByCenter}.UVs) {name=Circle UV, nodeColor=<rgba 0 0 0 1>}
  n8 = NamedRerouteUsage()
  n9 = NamedRerouteUsage()
  n10 = NamedRerouteUsage()
  n11 = NamedRerouteUsage()
  n12 = Multiply(
   A=ScalarParameter() {defaultValue=0.8, parameterName=Circle Opacity}
   B=Add(A=LinearInterpolate(A=@n8, B=@n9, Alpha=@n9), B=LinearInterpolate(A=@n10, B=@n11, Alpha=@n11))
  ) {constA=0.8}
  n13 = NamedRerouteUsage()
  NamedRerouteDeclaration(None=Multiply(A=Add(A=@n12, B=@n13), B=ScalarParameter() {defaultValue=1, parameterName=Master Opacity})) {name=Mask, nodeColor=<rgba 0 0 0 1>}
  n14 = LinearInterpolate(
   A=LinearInterpolate(B=VectorParameter() {parameterName=OutlineColor, defaultValue=<rgba 0.0176 0.5906 0.15 1>}.RGB, Alpha=NamedRerouteUsage())
   Alpha=NamedRerouteUsage()
  )
  n15 = Constant3Vector()
  n16 = NamedRerouteUsage()
  NamedRerouteDeclaration(None=LinearInterpolate(A=LinearInterpolate(A=@n14, B=@n15, Alpha=@n16), B=NamedRerouteUsage(), Alpha=NamedRerouteUsage())) {name=Final Color, nodeColor=<rgba 0 0 0 1>}
  n17 = MaterialFunctionCall(UVs=NamedRerouteUsage(), Size=Constant() {r=0.57}, OutlineThickness=Constant() {r=0.04}) {materialFunction=&MF_UI_Circle}
  NamedRerouteDeclaration(None=@n17.Outline) {name=Outer Outline, nodeColor=<rgba 0 0 0 1>}
  n18 = MaterialFunctionCall(UVs=NamedRerouteUsage(), Size=Constant() {r=0.5}, OutlineThickness=Constant() {r=0.025}) {materialFunction=&MF_UI_Circle}
  NamedRerouteDeclaration(None=@n18.Fill) {name=Inner Circle, nodeColor=<rgba 0 0 0 1>}
  NamedRerouteDeclaration(None=@n18.Outline) {name=Inner Outline, nodeColor=<rgba 0 0 0 1>}
  NamedRerouteDeclaration(None=@n17.Fill) {name=Outer Circle, nodeColor=<rgba 0 0 0 1>}
  NamedRerouteDeclaration(None=Multiply(A=@n4.RGB, B=@n4.RGB)) {name=IconColor, nodeColor=<rgba 0 0 0 1>}
  NamedRerouteUsage()
  NamedRerouteUsage()
 ]
}
