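# Translucent UI material for the NudgeHonk progress bar, drawing a rounded fill with animated decoration and a progress mark.
# Panning Perlin noise distorts two rotated box masks, forming the FX1 and FX2 layers.
# A Progress parameter steps horizontal UI UVs into a fill mask and thin edge marker, alongside a nine-sliced rounded-box mask.
# Named reroutes carry these masks into layered green tints, a tintable mark color and opacity scaled by Master Opacity.
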Material /FruitFire_NudgeHonk/Gameplay/ProgressBar/Materials/M_UI_NudgeHonk_Decoration
graph {
 properties {bAutomaticallySetUsageInEditor=false, blendMode=BLEND_Translucent, materialDomain=MD_UI, usedShadingModels=Default Lit}
 nodes [
  n1 = Add(A=TextureCoordinate(), B=Constant2Vector() {g=-0.1})
  n2 = TextureSampleParameter2D(UVs=Panner(Coordinate=TextureCoordinate()) {speedY=1}) {parameterName=NoiseTexture, texture=/FruitFire_NudgeHonk/UI/Textures/T_UI_RTL_PerlinNoise2}
  n3 = Constant() {r=0.6}
  n4 = AppendVector(A=ComponentMask(None=@n1) {r=true}, B=Add(A=ComponentMask(None=@n1) {g=true}, B=Multiply(A=@n2.RGB, B=@n3) {constB=0.056}))
  n5 = Constant() {r=0.25}
  NamedRerouteDeclaration(None=MaterialFunctionCall(
    UVs=MaterialFunctionCall(UVs=@n4, "Rotation Angle (0-1)"=@n5) {materialFunction=&CustomRotator}.Rotated Values
    Size=Constant2Vector() {g=1, r=0.5}
   ) {materialFunction=&MF_UI_Box}.Fill) {name=FX2, nodeColor=<rgba 0 0.2734 1 1>}
  n6 = Add(A=TextureCoordinate(), B=Constant2Vector() {g=-0.1})
  n7 = TextureSampleParameter2D(UVs=Panner(Coordinate=TextureCoordinate()) {speedY=1}) {parameterName=NoiseTexture, texture=&T_UI_RTL_PerlinNoise2}
  n8 = ScalarParameter() {defaultValue=0.5, parameterName=Intensity}
  n9 = AppendVector(A=ComponentMask(None=@n6) {r=true}, B=Add(A=ComponentMask(None=@n6) {g=true}, B=Multiply(A=@n7.RGB, B=@n8) {constB=0.056}))
  n10 = Constant() {r=0.25}
  NamedRerouteDeclaration(None=MaterialFunctionCall(
    UVs=MaterialFunctionCall(UVs=@n9, "Rotation Angle (0-1)"=@n10) {materialFunction=&CustomRotator}.Rotated Values
    Size=Constant2Vector() {g=1, r=0.2}
   ) {materialFunction=&MF_UI_Box}.Fill) {name=FX1, nodeColor=<rgba 0.2422 0 1 1>}
  n11 = NamedRerouteUsage()
  n12 = Multiply(B=NamedRerouteUsage()) {constA=1}
  n13 = Multiply(A=NamedRerouteUsage())
  n14 = Multiply(A=Multiply(A=Add(A=@n12, B=@n13), B=NamedRerouteUsage()), B=NamedRerouteUsage())
  NamedRerouteDeclaration(None=Multiply(A=Add(A=@n11, B=@n14), B=ScalarParameter() {defaultValue=1, parameterName=Master Opacity})) {name=opacity, nodeColor=<rgba 0.5234 0 1 1>}
  n15 = LinearInterpolate(B=Constant3Vector() {constant=<rgba 0.0905 0.8254 0.267 1>}, Alpha=NamedRerouteUsage())
  n16 = Constant3Vector() {constant=<rgba 0.457 1 0.583 1>}
  n17 = NamedRerouteUsage()
  NamedRerouteDeclaration(None=LinearInterpolate(
    A=LinearInterpolate(A=@n15, B=@n16, Alpha=@n17)
    B=VectorParameter() {parameterName=Progress Mark Color, defaultValue=<rgba 1 1 1 1>}.RGB
    Alpha=NamedRerouteUsage()
   )) {name=Final color, nodeColor=<rgba 0 1 0.2031 1>}
  n18 = MaterialFunctionCall(
   UVs=MaterialFunctionCall!(UVs) {materialFunction=/FortUILibrary/MaterialFunctions/Basics/MF_UI_9SliceUVs}.UVs
   Size=Constant2Vector() {g=0.85, r=0.85}
   CornerRoundness=Constant() {r=0.15}
  ) {materialFunction=&MF_UI_Box}
  NamedRerouteDeclaration(None=@n18.Fill) {name=BossHealthBar_Progress, nodeColor=<rgba 0 0 0 1>}
  n19 = Subtract(
   A=ComponentMask(None=MaterialFunctionCall() {materialFunction=&GetUserInterfaceUV}.Normalized UV) {r=true}
   B=ScalarParameter() {defaultValue=1, group=Progress, parameterName=Progress, sliderMax=1}
  )
  NamedRerouteDeclaration(None=OneMinus(None=Step(X=@n19) {constX=0})) {name=ProgressGradient, nodeColor=<rgba 0 0 0 1>}
  n20 = ScalarParameter() {defaultValue=1, group=Progress, parameterName=Progress, sliderMax=1}
  n21 = ComponentMask(None=MaterialFunctionCall() {materialFunction=&GetUserInterfaceUV}.Normalized UV) {r=true}
  n22 = Subtract(A=Step(Y=@n20, X=@n21), B=Step(Y=Add(A=@n20) {constB=0.006}, X=@n21))
  n23 = ScalarParameter() {defaultValue=1, parameterName=Show Progress Mark}
  NamedRerouteDeclaration(None=Multiply(A=LinearInterpolate(B=@n22, Alpha=@n23), B=NamedRerouteUsage())) {name=Progress Mark, nodeColor=<rgba 0 0 0 1>}
  NamedRerouteUsage()
  NamedRerouteUsage()
  NamedRerouteUsage()
  OneMinus(None=@n18.Fill)
 ]
}
