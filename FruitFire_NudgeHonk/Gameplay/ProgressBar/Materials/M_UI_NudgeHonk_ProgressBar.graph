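# Translucent UI material for the NudgeHonk progress bar, with fill, empty and FX colors, an optional progress mark and master opacity.
# 9-sliced rounded box functions build the backplate, inner frame and light outline masks.
# A Progress scalar steps against horizontal UI UVs for the fill edge and thin mark.
# Panning Perlin noise distorts rotated box masks into two animated FX bands, with named reroutes carrying masks into color and opacity.
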Material /FruitFire_NudgeHonk/Gameplay/ProgressBar/Materials/M_UI_NudgeHonk_ProgressBar
graph {
 properties {bAutomaticallySetUsageInEditor=false, blendMode=BLEND_Translucent, materialDomain=MD_UI, usedShadingModels=Default Lit}
 nodes [
  n1 = Add(A=TextureCoordinate(), B=Constant2Vector() {g=-0.2})
  n2 = ComponentMask(None=@n1) {r=true}
  n3 = Panner(Coordinate=TextureCoordinate()) {speedY=0.5}
  n4 = Add(
   A=ComponentMask(None=@n1) {g=true}
   B=Multiply(
    A=TextureSampleParameter2D(UVs=@n3) {parameterName=NoiseTexture, texture=/FruitFire_NudgeHonk/UI/Textures/T_UI_RTL_PerlinNoise2}.RGB
    B=Constant() {r=0.7}
   ) {constB=0.056}
  )
  n5 = MaterialFunctionCall(
   UVs=MaterialFunctionCall(UVs=AppendVector(A=@n2, B=@n4), "Rotation Angle (0-1)"=Constant() {r=0.25}) {materialFunction=&CustomRotator}.Rotated Values
   Size=Constant2Vector() {g=1, r=0.5}
  ) {materialFunction=&MF_UI_Box}
  Multiply(A=@n5.Fill, B=NamedRerouteUsage())
  NamedRerouteDeclaration(None=MaterialFunctionCall(
    UVs=MaterialFunctionCall!(UVs) {materialFunction=/FortUILibrary/MaterialFunctions/Basics/MF_UI_9SliceUVs}.UVs
    Size=Constant2Vector() {g=1, r=1}
    CornerRoundness=Constant() {r=0.3}
   ) {materialFunction=&MF_UI_Box}.Fill) {name=BossHealthBar_Backplate, nodeColor=<rgba 0 0 0 1>}
  n6 = MaterialFunctionCall() {materialFunction=&MF_UI_9SliceUVs}
  n7 = Constant2Vector() {g=0.85, r=0.85}
  n8 = Constant() {r=0.15}
  NamedRerouteDeclaration(None=OneMinus(None=MaterialFunctionCall(UVs=@n6.UVs, Size=@n7, CornerRoundness=@n8) {materialFunction=&MF_UI_Box}.Fill)) {name=BossHealthBar_Progress, nodeColor=<rgba 0 0 0 1>}
  NamedRerouteDeclaration(None=@n5.Fill) {name=FX2, nodeColor=<rgba 0 0.2734 1 1>}
  n9 = Subtract(
   A=ComponentMask(None=MaterialFunctionCall() {materialFunction=&GetUserInterfaceUV}.Normalized UV) {r=true}
   B=ScalarParameter() {defaultValue=0.5, group=Progress, parameterName=Progress, sliderMax=1}
  )
  NamedRerouteDeclaration(None=OneMinus(None=Step(X=@n9) {constX=0})) {name=ProgressGradient, nodeColor=<rgba 0 0 0 1>}
  n10 = LinearInterpolate(
   A=VectorParameter() {parameterName=EmptyColor, defaultValue=<rgba 0 0 0 1>}.RGB
   B=VectorParameter() {parameterName=FillColor, defaultValue=<rgba 0.0176 0.5906 0.15 1>}.RGB
   Alpha=NamedRerouteUsage()
  )
  n11 = VectorParameter() {parameterName=EmptyColor, defaultValue=<rgba 0 0 0 1>}
  n12 = NamedRerouteUsage()
  n13 = Multiply(A=NamedRerouteUsage(), B=ScalarParameter() {defaultValue=0.2, parameterName=Light Outline intensity}) {constB=0.1}
  n14 = NamedRerouteUsage()
  n15 = Multiply(A=NamedRerouteUsage(), B=OneMinus(None=NamedRerouteUsage()))
  n16 = LinearInterpolate(
   A=LinearInterpolate(A=LinearInterpolate(A=@n10, B=@n11.RGB, Alpha=@n12), Alpha=Multiply(A=@n13, B=@n14))
   B=VectorParameter() {parameterName=FX Color 1, defaultValue=<rgba 0.0297 0.7832 0.2011 1>}.RGB
   Alpha=Multiply(A=NamedRerouteUsage(), B=@n15)
  )
  n17 = VectorParameter() {parameterName=FX Color 2, defaultValue=<rgba 0.5934 1 0.6859 1>}
  n18 = Multiply(A=NamedRerouteUsage(), B=@n15)
  n19 = VectorParameter() {parameterName=Progress Mark Color, defaultValue=<rgba 0.5906 1 0.6867 1>}
  n20 = NamedRerouteUsage()
  NamedRerouteDeclaration(None=Add(A=LinearInterpolate(A=@n16, B=@n17.RGB, Alpha=@n18), B=Multiply(A=@n19.RGB, B=@n20))) {name=Final Color, nodeColor=<rgba 0 0 0 1>}
  n21 = ScalarParameter() {defaultValue=0.5, group=Progress, parameterName=Progress, sliderMax=1}
  n22 = ComponentMask(None=MaterialFunctionCall() {materialFunction=&GetUserInterfaceUV}.Normalized UV) {r=true}
  n23 = Subtract(A=Step(Y=@n21, X=@n22), B=Step(Y=Add(A=@n21) {constB=0.004}, X=@n22))
  n24 = ScalarParameter() {defaultValue=1, parameterName=Show Progress Mark}
  n25 = NamedRerouteUsage()
  NamedRerouteDeclaration(None=Multiply(A=LinearInterpolate(B=@n23, Alpha=@n24), B=OneMinus(None=@n25))) {name=Progress Mark, nodeColor=<rgba 0 0 0 1>}
  n26 = NamedRerouteUsage()
  n27 = NamedRerouteUsage()
  n28 = Multiply(A=@n27) {constB=0.4}
  n29 = NamedRerouteUsage()
  n30 = NamedRerouteUsage()
  n31 = NamedRerouteUsage()
  n32 = Multiply(A=NamedRerouteUsage(), B=OneMinus(None=NamedRerouteUsage()))
  n33 = Add(
   A=Multiply(A=@n27, B=LinearInterpolate(A=@n28, B=@n29, Alpha=@n29) {constAlpha=0})
   B=Multiply(A=LinearInterpolate(A=@n30, B=@n31, Alpha=@n32), B=@n32)
  )
  NamedRerouteDeclaration(None=Multiply(A=Add(A=@n26, B=@n33), B=ScalarParameter() {defaultValue=1, parameterName=Master Opacity})) {name=opacity, nodeColor=<rgba 0 0.4609 1 1>}
  n34 = Add(A=TextureCoordinate(), B=Constant2Vector() {g=-0.1})
  n35 = TextureSampleParameter2D(UVs=Panner(Coordinate=TextureCoordinate()) {speedX=0.4, speedY=1}) {parameterName=NoiseTexture, texture=&T_UI_RTL_PerlinNoise2}
  n36 = Constant() {r=0.5}
  n37 = AppendVector(A=ComponentMask(None=@n34) {r=true}, B=Add(A=ComponentMask(None=@n34) {g=true}, B=Multiply(A=@n35.RGB, B=@n36) {constB=0.056}))
  n38 = Constant() {r=0.25}
  NamedRerouteDeclaration(None=MaterialFunctionCall(
    UVs=MaterialFunctionCall(UVs=@n37, "Rotation Angle (0-1)"=@n38) {materialFunction=&CustomRotator}.Rotated Values
    Size=Constant2Vector() {g=1, r=0.12}
   ) {materialFunction=&MF_UI_Box}.Fill) {name=FX1, nodeColor=<rgba 0 0.2734 1 1>}
  NamedRerouteDeclaration(None=MaterialFunctionCall(
    UVs=MaterialFunctionCall() {materialFunction=&MF_UI_9SliceUVs}.UVs
    Size=Constant2Vector() {g=0.7, r=0.7}
    CornerRoundness=Constant() {r=0.02}
    OutlineThickness=Constant() {r=0.06}
   ) {materialFunction=&MF_UI_Box}.Outline) {name=LightOutline, nodeColor=<rgba 0 0 0 1>}
  NamedRerouteUsage()
  NamedRerouteUsage()
 ]
}
