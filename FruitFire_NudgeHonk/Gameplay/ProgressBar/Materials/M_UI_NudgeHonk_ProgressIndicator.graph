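# Translucent UI material for the NudgeHonk progress bar indicator, with a flat green color constant and a Master Opacity control.
# Width and Height parameters form an aspect ratio that scales horizontal UVs to keep the detail texture undistorted.
# Progress offsets those UVs, with small edge corrections, to slide the clamped BarDetails texture along the bar.
# The texture's alpha becomes the indicator shape through named reroutes.
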Material /FruitFire_NudgeHonk/Gameplay/ProgressBar/Materials/M_UI_NudgeHonk_ProgressIndicator
graph {
 properties {bAutomaticallySetUsageInEditor=false, blendMode=BLEND_Translucent, materialDomain=MD_UI, usedShadingModels=Default Lit}
 nodes [
  Constant3Vector() {constant=<rgba 0.0152 0.4969 0.1274 1>}
  Multiply(A=NamedRerouteUsage(), B=ScalarParameter() {defaultValue=1, parameterName=Master Opacity})
  NamedRerouteDeclaration(None=Divide(A=ScalarParameter() {defaultValue=640, parameterName=Widgth}, B=ScalarParameter() {defaultValue=64, parameterName=Height})) {name=AspectRatio, nodeColor=<rgba 1 0 0.6094 1>}
  n1 = TextureCoordinate()
  n2 = ComponentMask(None=@n1) {r=true}
  n3 = NamedRerouteUsage()
  NamedRerouteDeclaration(None=AppendVector(A=Multiply(A=@n2, B=@n3), B=ComponentMask(None=@n1) {g=true})) {name=UV_Fixed, nodeColor=<rgba 0 0.9297 1 1>}
  n4 = NamedRerouteUsage()
  n5 = ComponentMask(None=@n4) {r=true}
  n6 = ScalarParameter() {defaultValue=1, group=Progress, parameterName=Progress, sliderMax=1}
  n7 = Subtract(A=NamedRerouteUsage())
  n8 = ScalarParameter() {defaultValue=1, group=Progress, parameterName=Progress, sliderMax=1}
  n9 = Add(A=Multiply(A=Multiply(A=@n6, B=@n7)) {constB=-1}, B=Multiply(A=Constant() {r=0.12}, B=LinearInterpolate(Alpha=@n8) {constA=1, constB=-1})) {constB=-0.232}
  n10 = TextureSample(UVs=AppendVector(A=Add(A=@n5, B=@n9), B=ComponentMask(None=@n4) {g=true})) {samplerSource=SSM_Clamp_WorldGroupSettings, texture=/FruitFire_NudgeHonk/Gameplay/ProgressBar/Texture/T_UI_FruitFire_NudgeHonk_BarDetails}
  NamedRerouteDeclaration(None=@n10.A) {name=FinalShape, nodeColor=<rgba 0 0.25 1 1>}
 ]
}
